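# Sound cue for the game-start sound of the computer activation game in a Season 25 Part 3 story quest (Q04).
# A single wave player plays the GameStart wave at slightly reduced volume through the Fort_Quests sound class.
# A concurrency override allows one instance per owner and blocks new plays while it is active.
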SoundCue /BattlePassS25/Items/QuestActors/Stories/Part03/ComputerActivationGame/SFX/S25_P03_Q04_GameStart_Cue
root=WavePlayer_0
nodes [{
  id=WavePlayer_0
  properties {soundWaveAssetPtr=&/S25_P03_Q04_GameStart}
 }]
properties {
 bOverrideConcurrency=true
 maxDistance=2097152
 soundClassObject=&Fort_Quests
 volumeMultiplier=0.8
 concurrencyOverrides {bLimitToOwner=true, maxCount=1, resolutionRule=PreventNew}
}
