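# Sound cue for the looping audio of the activated table in the Season 25 Battle Pass Part 03 Q04 computer activation quest game.
# A single looping wave player feeds the TableActivated loop wave straight to the output.
# It plays under the Fort_Quests sound class at slightly reduced volume, using the quest table's attenuation asset for positional falloff.
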SoundCue /BattlePassS25/Items/QuestActors/Stories/Part03/ComputerActivationGame/SFX/S25_P03_Q04_TableActivated_LP_Cue
root=WavePlayer_0
nodes [{
  id=WavePlayer_0
  properties {bLooping=true, soundWaveAssetPtr=&/S25_P03_Q04_TableActivated_LP}
 }]
properties {
 attenuationSettings=&/S25_P03_Q04_Table_Atten
 maxDistance=1700
 soundClassObject=&Fort_Quests
 volumeMultiplier=0.85
}
